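# Conversation database for the ShadowOps NPC, providing the greeting shown when a conversation with that character starts.
# The entry point leads to a single speech node with a default line and contextual alternates prioritized by curve rows for time since played, first-ever interaction and high Bars total.
# The speech node then branches through link nodes to the shared NPC Library quests and basic services conversations.
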ConversationDatabase /NPCLibrary/NPCs/ShadowOps/ShadowOps/NPCLibrary_Conversation_ShadowOps
entry_points [{tag=Conversation.Entry.ShadowOps, destinations [#1]}]
nodes [
 ConversationEntryPointNode_7 {
  id=#1
  class=ConversationEntryPointNode
  outputs [#2]
  properties {entryTag {tagName=Conversation.Entry.ShadowOps}}
 }
 Speech_C_4 {
  id=#2
  class=Speech_C
  outputs [#3, #4]
  properties {
   nodeName=""
   generalConfig {
    defaultMessage="Ready to get tactical? "
    contextualMessages [
     {
      message=Your muscle memory should kick back in pretty soon.
      contextRequirements [{requirement=null}]
      priority {curve {rowName=Tandem.Speech.Priority.WhileSincePlayed}}
     }
     {
      message="I've heard about your combat style! Impressive."
      contextRequirements [{requirement=null}]
      priority {curve {rowName=Tandem.Speech.Priority.FirstEverInteraction}}
     }
     {
      message="I think J.B.'s accepting donations if you want to offload those Bars!"
      contextRequirements [{requirement=null}]
      priority {curve {rowName=Tandem.Speech.Priority.HighTotalBars}}
     }
    ]
   }
  }
 }
 ConversationLinkNode_0 {id=#4, class=ConversationLinkNode, properties {remoteEntryTag {tagName=Conversation.Entry.BasicServices}}}
 ConversationLinkNode_1 {id=#3, class=ConversationLinkNode, properties {remoteEntryTag {tagName=Conversation.Entry.NPCLibrary.Quests}}}
]
